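# MFC_ShakeCrunch_SteppedTime is a material function that produces looping, stepped time values for the ShakeCrunch character body materials.
# Time is scaled, wrapped with Frac, multiplied by a step count and floored into discrete step indices.
# It outputs that index as FlooredOutput, and the index multiplied by an offset parameter as TimeStep.
# Scale, step count and offset are scalar parameters in a shared time-stepping group.
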MaterialFunction /BRCosmetics/Characters/Player/Male/Medium/Bodies/M_MED_ShakeCrunch/Materials/MFC_ShakeCrunch_SteppedTime
graph {nodes [
  n1 = Multiply(A=Time(), B=ScalarParameter() {defaultValue=0.2, group=03_ShakeCrunch_TimeStepping, parameterName=SteppedTime_Scale, sliderMax=1})
  n2 = Floor(None=Multiply(A=Frac(None=@n1), B=ScalarParameter() {defaultValue=12, group=03_ShakeCrunch_TimeStepping, parameterName=SteppedTime_Steps, sliderMax=32}))
  FunctionOutput(None=Multiply(
    A=FunctionOutput(None=@n2) {outputName=FlooredOutput, sortPriority=1}
    B=ScalarParameter() {defaultValue=0.125, group=03_ShakeCrunch_TimeStepping, parameterName=SteppedTime_Offset, sliderMax=1}
   )) {outputName=TimeStep}
 ]}
