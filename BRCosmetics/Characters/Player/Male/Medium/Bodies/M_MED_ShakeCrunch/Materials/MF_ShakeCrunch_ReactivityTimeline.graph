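# Material function for the ShakeCrunch player body that turns a trigger time and animation duration into a reactivity timeline pulse.
# Elapsed time since Time_In is divided by AnimDuration, saturated, then fed through a quarter-shifted sine.
# A ceil threshold and inversion make the sole output, Round, a hard gate: one during the animation, zero at and near both ends.
# Power-shaped curve branches and a UV debug mask remain unconnected to the output.
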MaterialFunction /BRCosmetics/Characters/Player/Male/Medium/Bodies/M_MED_ShakeCrunch/Materials/MF_ShakeCrunch_ReactivityTimeline
graph {nodes [
  ComponentMask(None=TextureCoordinate()) {desc=debug curve, r=true}
  n1 = Time()
  n2 = FunctionInput(Preview=Constant() {r=-99999}) {bUsePreviewValueAsDefault=true, inputName=Time_In, inputType=FunctionInput_Scalar}
  n3 = FunctionInput(Preview=Constant() {r=3}) {bUsePreviewValueAsDefault=true, inputName=AnimDuration, inputType=FunctionInput_Scalar, sortPriority=1, previewValue=<4 0 0 1>}
  n4 = Saturate(None=Divide(A=Subtract(A=@n1, B=@n2), B=Add(A=@n3) {constB=0}))
  n5 = Sine(None=Frac(None=Add(A=@n4) {constB=0.25}))
  n6 = Ceil(None=Add(A=@n5) {constB=-0.99})
  FunctionOutput(None=OneMinus(None=Saturate(None=@n6))) {outputName=Round}
  n7 = Add(A=Multiply(A=@n5) {constB=0.5}) {constB=0.5}
  n8 = Multiply(A=@n7, B=@n7)
  n9 = Multiply(A=@n8, B=@n8) {desc=bot curve pow}
  n10 = OneMinus(None=Multiply(A=@n9, B=@n9))
  Round(None=Multiply(A=Multiply(A=@n10, B=@n10), B=@n10) {desc=top curve pow})
  Round(None=@n10)
 ]}
