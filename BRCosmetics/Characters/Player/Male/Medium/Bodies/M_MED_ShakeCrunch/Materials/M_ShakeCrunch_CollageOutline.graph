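# Masked, unlit inverted-hull outline material for the ShakeCrunch character, styled as animated paper collage.
# WPO pushes vertices along surface, baked or normal-map normals, with switchable thickness mask and distance-based width.
# Opacity keeps back faces, broken up by Fresnel, stepped-time noise and camera fade, then temporally dithered.
# Color overlays stepped-time paper texture with collage and glitch tints, shaded by a screen-space gradient, alongside a reactive animation timeline.
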Material /BRCosmetics/Characters/Player/Male/Medium/Bodies/M_MED_ShakeCrunch/Materials/M_ShakeCrunch_CollageOutline
graph {
 properties {
  bAutomaticallySetUsageInEditor=false
  bUseMaterialAttributes=true
  bUsedWithMeshDeformer=true
  bUsedWithMorphTargets=true
  bUsedWithSkeletalMesh=true
  bUsedWithStaticLighting=true
  blendMode=BLEND_Masked
  shadingModel=MSM_Unlit
  twoSided=true
  usedShadingModels=Unlit
 }
 nodes [
  Composite() {subgraphName=Construct Base Inputs}
  n1 = NamedRerouteUsage()
  MaterialFunctionCall!(In=@n1) {materialFunction=/Game/Packages/Fortress_SharedMaterials/Anime_Master_Material/MF_TODCustomColors}
  MaterialFunctionCall!(Position, UVRadialScale) {materialFunction=/Game/Characters/Player/Common/MaterialFunctions/Superlevel/CharacterStabilizedScreenSpaceUVs}
  n2 = NamedRerouteUsage()
  n3 = NamedRerouteUsage()
  n4 = MaterialFunctionCall!(Alpha Threshold=@n3, Random) {materialFunction=/Game/Environments/Asteria/Sets/NeonCity/Techart/Screen/MaterialFunctions/DitherTemporalAABlue}
  n5 = ScalarParameter() {defaultValue=1, group=01_InvertedHull_Outlines, parameterName=OpacityMask_Dithering_Intensity, sliderMax=2}
  n6 = NamedRerouteUsage()
  n7 = Constant()
  MaterialExpressionReroute_0 = Reroute(None=@n7)
  n8 = PreSkinnedPosition()
  n9 = ShaderStageSwitch(PixelShader=VertexInterpolator(VS=@n8).PS, VertexShader=@n8)
  MaterialExpressionReroute_5 = Reroute(None=@n9)
  n10 = MaterialFunctionCall!(Material) {materialFunction=/Game/Packages/Fortress_SharedMaterials/Base_Material_Functions/CharacterBrightnessandScale_Balance}
  MaterialExpressionReroute_6 = Reroute(None=@n10.Distance Blend)
  n11 = Fresnel() {baseReflectFraction=0, exponent=1}
  MaterialExpressionReroute_7 = Reroute(None=@n11)
  MaterialFunctionCall(
   Material=SetMaterialAttributes!(MaterialAttributes, Emissive Color=@n2, Opacity Mask=LinearInterpolate(A=@n3, B=@n4.Result, Alpha=@n5), World Position Offset=@n6) {attributeSetTypes [#1, #2, #3]}
   M=@MaterialExpressionReroute_0.M
   Pre Skinned Local Position=@MaterialExpressionReroute_5.Pre Skinned Local Position
   Distance Blend=@MaterialExpressionReroute_6.Distance Blend
   Fresnel=@MaterialExpressionReroute_7.Fresnel
  ) {materialFunction=&MF_CharacterParent_2023_PostFX}
  n12 = ScalarParameter() {defaultValue=1000, group=01_InvertedHull_Outlines, parameterName=Fade_Distance, sliderMax=1500, sliderMin=500, sortPriority=5}
  NamedRerouteDeclaration(None=MaterialFunctionCall!(
    Fade Offset=ScalarParameter() {defaultValue=50, group=01_InvertedHull_Outlines, parameterName=Fade_Offset, sliderMax=100, sortPriority=5}
    Fade Distance=@n12
   ) {materialFunction=/Game/Global/Materials/Character_Parent_2023/Anime_Parent_2023/MaterialFunctions/MFC_DistanceFade_FOV_compensation}.Result) {name=DistanceFading, nodeColor=<rgba 1 0.3047 0 1>}
  n13 = NamedRerouteUsage()
  n14 = Abs(None=DotProduct(A=CameraVectorWS(), B=Transform(None=@n13) {transformSourceType=TRANSFORMSOURCE_Local}))
  NamedRerouteDeclaration(None=Power(
    Base=OneMinus(None=@n14)
    Exp=ScalarParameter() {defaultValue=5, group=01_InvertedHull_Outlines, parameterName=BakedNormalOpacity_Exponent, sliderMax=12}
   )) {name=BakedNormalFresnel, nodeColor=<rgba 0 0.8144 1 1>}
  NamedRerouteDeclaration(None=MaterialFunctionCall() {materialFunction=&/MFC_ShakeCrunch_SteppedTime}.TimeStep) {name=TimeStep, nodeColor=<rgba 0.052 1 0.9901 1>}
  n15 = ScalarParameter() {defaultValue=3, group=Shoes, parameterName=Shoes_MaskHeight}
  n16 = Add(A=@n15, B=ScalarParameter() {defaultValue=20, group=Shoes, parameterName=Shoes_MaskWidth})
  n17 = ComponentMask(None=VertexInterpolator(VS=LocalPosition() {localOrigin=InstancePreSkinning}.XYZ).PS) {b=true}
  NamedRerouteDeclaration(None=LinearInterpolate(B=SmoothStep(Min=@n15, Max=@n16, Value=@n17) {constMax=160}, Alpha=ScalarParameter() {group=Shoes, parameterName=UseShoes}) {constA=1}) {name=ShoesMaskMultiplier, nodeColor=<rgba 1 0 0.0241 1>}
  n18 = ScalarParameter() {defaultValue=0.5, desc="LineThickness (Near)", group=01_InvertedHull_Outlines, parameterName=LineThickness, sliderMax=10, sortPriority=3}
  n19 = StaticSwitchParameter(
   True=LinearInterpolate(
    A=@n18
    B=ScalarParameter() {defaultValue=1.5, group=01_InvertedHull_Outlines, parameterName=LineThickness_Far, sliderMax=10, sortPriority=3}
    Alpha=NamedRerouteUsage()
   )
   False=@n18
  ) {defaultValue=true, group=01_InvertedHull_Outlines, parameterName=UseDistanceFading, sortPriority=0}
  n20 = Multiply(A=NamedRerouteUsage(), B=@n19)
  n21 = WorldPosition()
  n22 = CameraPositionWS()
  n23 = Multiply(A=Normalize(VectorInput=Subtract(A=@n21.XYZ, B=@n22))) {constB=-1}
  n24 = ScalarParameter() {group=01_InvertedHull_Outlines, parameterName=CameraDirection_Push, sliderMax=50, sliderMin=-50}
  n25 = ChannelMaskParameter(None=VertexColor()) {group=01_InvertedHull_Outlines, maskChannel=Green, parameterName=CameraDirection_ChannelMask, defaultValue=<rgba 0 1 0 0>}
  n26 = OneMinus(None=Multiply(A=@n25.R, B=ScalarParameter() {group=01_InvertedHull_Outlines, parameterName=CameraDirection_Strenght, sliderMax=2}))
  n27 = Add(
   A=Multiply(
    A=StaticSwitchParameter(True=@n20, False=@n19) {group=01_InvertedHull_Outlines, parameterName=UseThicknessMap, sortPriority=0}
    B=NamedRerouteUsage()
   )
   B=Multiply(A=Multiply(A=@n23, B=@n24) {constB=0}, B=NamedRerouteDeclaration(None=@n26) {name=VertexMask, nodeColor=<rgba 0 0.7187 1 1>})
  )
  NamedRerouteDeclaration(None=StaticSwitchParameter(
    True=MaterialFunctionCall!(WPO=@n27) {materialFunction=/Game/Packages/Fortress_SharedMaterials/Base_Material_Functions/Fortnite_HideElementsWithVertexColor_2022}.Result
    False=@n27
   ) {group=03_Opacity Mask, parameterName=Use Vertex Colors for Mask}) {name=WPO, nodeColor=<rgba 1 0.3281 0 1>}
  NamedRerouteDeclaration(None=MaterialFunctionCall() {materialFunction=/Game/Global/Materials/Character_Parent_2023/Anime_Parent_2023/MaterialFunctions/MFC_ActorAlighnedScreenPosition}.UVs) {name=ObjectAlighedScreenPosition, nodeColor=<rgba 1 0.867 0 1>}
  NamedRerouteDeclaration(None=MaterialFunctionCall!(
    Time_In=ScalarParameter() {defaultValue=-99999, group=03_ShakeCrunch_DotsRimHatching, parameterName=Reactive_Collage_CT, sliderMin=-99999}
    AnimDuration=ScalarParameter() {defaultValue=4, group=03_ShakeCrunch_DotsRimHatching, parameterName=Reactive_Collage_AnimDuration, sliderMin=0}
   ) {materialFunction=&/MF_ShakeCrunch_ReactivityTimeline}.Round) {name=AnimGradient_Reactive, nodeColor=<rgba 1 0 0.3047 1>}
  n28 = TextureSampleParameter2D() {
   group=01_InvertedHull_Outlines
   parameterName=FX
   samplerType=SAMPLERTYPE_Masks
   sortPriority=1
   texture=/Game/Packages/Fortress_SharedMaterials/Anime_Master_Material/T_Outline_Mask
  }
  n29 = ChannelMaskParameter(None=@n28.RGB) {group=01_InvertedHull_Outlines, maskChannel=Blue, parameterName=OutlineThickness_Channel, sortPriority=3, defaultValue=<rgba 0 0 1 0>}
  n30 = Constant() {desc=one bit used for opacity cutout, r=-0.009}
  n31 = Ceil(None=NamedRerouteDeclaration(None=Add(A=@n29.R, B=@n30)) {name=ThicknessMask, nodeColor=<rgba 1 0.1689 0 1>}) {desc=cut outlines mask}
  n32 = OneMinus(None=TwoSidedSign())
  n33 = Add(
   A=Fresnel() {baseReflectFraction=0, exponent=1}
   B=ScalarParameter() {group=01_InvertedHull_Outlines, parameterName=Fresnel_Offset, sliderMax=1, sliderMin=-1}
  )
  n34 = Multiply(
   A=Power(Base=OneMinus(None=@n33), Exp=ScalarParameter() {defaultValue=2, group=01_InvertedHull_Outlines, parameterName=Fresnel_Exp, sliderMax=8})
   B=Constant() {r=2}
  )
  n35 = NamedRerouteUsage()
  n36 = MaterialFunctionCall() {materialFunction=&MFC_ShakeCrunch_SteppedTime}
  n37 = TextureSampleParameter2D(UVs=Frac(None=Add(A=@n35, B=@n36.TimeStep))) {parameterName=NoiseMap, texture=&AtmosphericCloudNoise01}
  n38 = ScalarParameter() {group=01_InvertedHull_Outlines, parameterName=Noise_Offset, sliderMax=1, sliderMin=-1}
  n39 = Subtract(A=NamedRerouteDeclaration(None=Add(A=@n37.R, B=@n38)) {name=Noise, nodeColor=<rgba 0.2656 0 1 1>}, B=Constant() {r=1})
  n40 = Multiply(
   A=StaticSwitchParameter(True=Multiply(A=@n31, B=@n32), False=@n32) {group=01_InvertedHull_Outlines, parameterName=UseThicknessMap, sortPriority=0}
   B=Saturate(None=Add(A=@n34, B=@n39)) {desc=height blend with 1 heighmap}
  )
  n41 = MaterialFunctionCall(Fade Length=Constant() {r=80}, Fade Offset=Constant() {r=30}) {materialFunction=&CameraDepthFade}
  NamedRerouteDeclaration(None=Multiply(A=Multiply(A=@n40, B=@n41.Result), B=NamedRerouteUsage())) {name=OpacityMask, nodeColor=<rgba 0 0.8125 1 1>}
  n42 = TextureSampleParameter2D() {
   group=01_InvertedHull_Outlines
   parameterName=NormalMap
   samplerType=SAMPLERTYPE_Normal
   sortPriority=1
   texture=&FlatNormalBlue
  }
  n43 = ScalarParameter() {
   defaultValue=-0.1
   desc="values above 1.0 and below 0.0 can help with seperated geometry.\r\n-0.1 seems like a good wellyou to alight split geo."
   group=01_InvertedHull_Outlines
   parameterName=NormalMap_Flatness
   sliderMax=2
   sliderMin=-1
   sortPriority=5
  }
  n44 = MaterialFunctionCall(Normal=@n42.RGB, Flatness=@n43) {materialFunction=&FlattenNormal}
  n45 = VertexColor()
  n46 = Normalize(VectorInput=Subtract(A=Multiply(A=@n45) {constB=2}))
  n47 = StaticSwitchParameter(
   True=Transform(None=NamedRerouteDeclaration(None=@n46) {name=BakedNormals, nodeColor=<rgba 0 0.2789 1 1>}) {transformSourceType=TRANSFORMSOURCE_Local}
   False=VertexNormalWS()
  ) {
   desc=use baked normals in the vertex colors instead of the surface normals.
   group=01_InvertedHull_Outlines
   parameterName=UseBakedNormals
   sortPriority=0
  }
  n48 = StaticSwitchParameter(True=Transform(None=Normalize(VectorInput=@n44.Result)), False=@n47) {desc=use normal map to prevent segmented geometry, group=01_InvertedHull_Outlines, parameterName=UseNormalMap, sortPriority=0}
  NamedRerouteDeclaration(None=@n48) {name=Normal, nodeColor=<rgba 0.224 0.224 1 1>}
  n49 = Multiply(
   A=Add(
    A=NamedRerouteUsage()
    B=MaterialFunctionCall() {materialFunction=&MFC_ShakeCrunch_SteppedTime}.TimeStep
   )
   B=ScalarParameter() {defaultValue=0.5, group=03_ShakeCrunch_DotsRimHatching, parameterName=PaperDetail_Tiling, sliderMax=2}
  )
  n50 = ChannelMaskParameter(None=TextureSample(UVs=Frac(None=@n49)) {texture=/BRCosmetics/Characters/Player/Male/Medium/Bodies/M_MED_ShakeCrunch/Textures/T_ShakeCrunch_Paper}.RGB) {group=03_ShakeCrunch_DotsRimHatching, maskChannel=Blue, parameterName=PaperDetail_Channel, defaultValue=<rgba 0 0 1 0>}
  n51 = VectorParameter() {group=01_InvertedHull_Outlines, parameterName=Collage_Color, sortPriority=2, defaultValue=<rgba 0.962 1 0.4375 1>}
  n52 = VectorParameter() {group=01_InvertedHull_Outlines, parameterName=CollageGlitch_Color, sortPriority=2, defaultValue=<rgba 1 0.1166 0 1>}
  n53 = MaterialFunctionCall(
   Base=Power(Base=@n50.R, Exp=ScalarParameter() {defaultValue=1.5, group=03_ShakeCrunch_DotsRimHatching, parameterName=PaperDetail_Exponent, sliderMax=8})
   Blend=LinearInterpolate(A=@n51.RGB, B=@n52.RGB, Alpha=NamedRerouteUsage())
  ) {materialFunction=&Blend_Overlay}
  n54 = ComponentMask(None=NamedRerouteUsage()) {g=true}
  n55 = Saturate(None=Multiply(A=Add(A=@n54) {constB=0.5}) {constB=0.5})
  n56 = Add(A=Power(Base=OneMinus(None=@n55)) {constExponent=2.2}) {constB=0.8, desc=ScreenSpaceGradient}
  NamedRerouteDeclaration(None=Saturate(None=Multiply(A=@n53.Result, B=@n56))) {name=ColorRaw, nodeColor=<rgba 0 0.5781 1 1>}
  NamedRerouteDeclaration(None=MaterialFunctionCall!(Color=MaterialFunctionCall!(In=@n1, LUT, LUT Multiplier) {materialFunction=/Game/Packages/Fortress_SharedMaterials/Anime_Master_Material/MF_ApplyLUT}.Result) {materialFunction=/Game/Packages/Fortress_SharedMaterials/Anime_Master_Material/MF_TODColorAdjustment}.Result) {name=ColorFinal, nodeColor=<rgba 1 0.6797 0 1>}
  NamedRerouteUsage()
  PinBase!(M=@n7, Pre Skinned Local Position=@n9, Distance Blend=@n10.Distance Blend, Fresnel=@n11) {reroutePins [
    M {expression=&self:MaterialExpressionReroute_0}
    Pre Skinned Local Position {expression=&self:MaterialExpressionReroute_5}
    Distance Blend {expression=&self:MaterialExpressionReroute_6}
    Fresnel {expression=&self:MaterialExpressionReroute_7}
   ]}
  PinBase!()
  n57 = TextureSampleParameter2D() {
   desc="Outline Color Texture, use can use the ShadedMap."
   group=01_InvertedHull_Outlines
   parameterName=ColorMap
   sortPriority=1
   texture=&T_BlacksRGB
  }
  n58 = ScalarParameter() {
   defaultValue=1
   desc=Multiplier for the Outline Color Texture
   group=01_InvertedHull_Outlines
   parameterName=Color_Intensity
   sliderMax=1
   sortPriority=3
  }
  StaticSwitchParameter(True=Multiply(A=@n57.RGB, B=@n58)) {
   desc=Enable to use a texture to color the edge outlines of the mesh.
   group=01_InvertedHull_Outlines
   parameterName=UseColorMap
   sortPriority=0
  }
  n59 = SetMaterialAttributes(World Position Offset=NamedRerouteUsage()) {
   desc=Debug
   attributeSetTypes [#1, #2, #3]
  }
  StaticSwitchParameter(
   True=@n59
   False=SetMaterialAttributes(Emissive Color=@n2, Opacity Mask=@n3, World Position Offset=@n6) {attributeSetTypes [#1, #2, #3]}
  ) {group=01_InvertedHull_Outlines, parameterName=DebugThicknessMap, sortPriority=90}
 ]
}
