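# Unlit, masked, two-sided material drawing inverted-hull outlines for the ShakeCrunch character.
# World position offset pushes vertices along surface or vertex-color-baked normals, with switches for thickness map, FOV-compensated distance width, camera-direction offset and vertex-color hiding.
# TwoSidedSign masks away front faces, and the thickness map can cut out outlines.
# Outline color passes through LUT and time-of-day functions into emissive, then attributes enter the shared character PostFX function.
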Material /BRCosmetics/Characters/Player/Male/Medium/Bodies/M_MED_ShakeCrunch/Materials/M_ShakeCrunch_InvHullOutline
graph {
 expressions {
  MaterialExpressionAbs_1 {class=MaterialExpressionAbs, inputs [None {from=MaterialExpressionDotProduct_0, output=""}], properties {}}
  MaterialExpressionAdd_0 {
   class=MaterialExpressionAdd
   inputs [
    A {from=MaterialExpressionMultiply_0}
    B {from=MaterialExpressionReroute_5}
   ]
  }
  MaterialExpressionAdd_3 {
   class=MaterialExpressionAdd
   inputs [
    A {from=MaterialExpressionChannelMaskParameter_1, output=R}
    B {from=MaterialExpressionConstant_0}
   ]
  }
  MaterialExpressionAdd_4 {
   class=MaterialExpressionAdd
   inputs [
    A {from=MaterialExpressionWorldPosition_1, output=XYZ}
    B {from=MaterialExpressionConstant3Vector_1}
   ]
  }
  MaterialExpressionCameraPositionWS_0 {class=MaterialExpressionCameraPositionWS, inputs [], properties {}}
  MaterialExpressionCameraVectorWS_0 {class=MaterialExpressionCameraVectorWS, inputs [], properties {}}
  MaterialExpressionCeil_0 {
   class=MaterialExpressionCeil
   inputs [None {from=MaterialExpressionNamedRerouteDeclaration_5, output=""}]
   properties {desc=cut outlines mask}
  }
  MaterialExpressionChannelMaskParameter_1 {
   class=MaterialExpressionChannelMaskParameter
   inputs [None {from=MaterialExpressionTextureSampleParameter2D_3, output=RGB}]
   properties {group=01_InvertedHull_Outlines, maskChannel=Blue, parameterName=OutlineThickness_Channel, sortPriority=3, defaultValue=<rgba 0 0 1 0>}
  }
  MaterialExpressionComposite_0 {class=MaterialExpressionComposite, inputs [], properties {subgraphName=Construct Base Inputs}}
  MaterialExpressionConstant3Vector_1 {class=MaterialExpressionConstant3Vector}
  MaterialExpressionConstant_0 {class=MaterialExpressionConstant, properties {desc=one bit used for opacity cutout, r=-0.009}}
  MaterialExpressionConstant_1 {class=MaterialExpressionConstant}
  MaterialExpressionDotProduct_0 {
   class=MaterialExpressionDotProduct
   inputs [
    A {from=MaterialExpressionCameraVectorWS_0, output=""}
    B {from=MaterialExpressionTransform_2, output=""}
   ]
   properties {}
  }
  MaterialExpressionFresnel_0 {
   class=MaterialExpressionFresnel
   inputs [
    ExponentIn {from="", output=""}
    BaseReflectFractionIn {from="", output=""}
    Normal {from="", output=""}
   ]
   properties {baseReflectFraction=0, exponent=1}
  }
  MaterialExpressionIf_0 {
   class=MaterialExpressionIf
   inputs [
    A {from=MaterialExpressionSaturate_0, output=""}
    B {from="", output=""}
    "A > B" {from=MaterialExpressionSubtract_1, output=""}
    "A == B" {from="", output=""}
    "A < B" {from=MaterialExpressionStaticSwitchParameter_8, output=""}
   ]
   properties {constB=1}
  }
  MaterialExpressionLinearInterpolate_8 {
   class=MaterialExpressionLinearInterpolate
   inputs [
    A {from=MaterialExpressionReroute_4}
    B {from=MaterialExpressionScalarParameter_40}
    Alpha {from=MaterialExpressionNamedRerouteUsage_0}
   ]
  }
  MaterialExpressionMaterialFunctionCall_13 {
   class=MaterialExpressionMaterialFunctionCall
   inputs [
    In {from=MaterialExpressionNamedRerouteUsage_3}
    LUT {from=""}
    LUT Multiplier {from=""}
   ]
   properties {materialFunction=/Game/Packages/Fortress_SharedMaterials/Anime_Master_Material/MF_ApplyLUT}
  }
  MaterialExpressionMaterialFunctionCall_14 {
   class=MaterialExpressionMaterialFunctionCall
   inputs [Material {from=""}]
   properties {materialFunction=/Game/Packages/Fortress_SharedMaterials/Base_Material_Functions/CharacterBrightnessandScale_Balance}
  }
  MaterialExpressionMaterialFunctionCall_16 {
   class=MaterialExpressionMaterialFunctionCall
   inputs [
    Material {from=MaterialExpressionSetMaterialAttributes_0}
    M {from=MaterialExpressionReroute_8, output=M}
    Pre Skinned Local Position {from=MaterialExpressionReroute_9, output=Pre Skinned Local Position}
    Distance Blend {from=MaterialExpressionReroute_10, output=Distance Blend}
    Fresnel {from=MaterialExpressionReroute_11, output=Fresnel}
   ]
   properties {materialFunction=&MF_CharacterParent_2023_PostFX}
  }
  MaterialExpressionMaterialFunctionCall_2 {
   class=MaterialExpressionMaterialFunctionCall
   inputs [WPO {from=MaterialExpressionReroute_3}]
   properties {materialFunction=/Game/Packages/Fortress_SharedMaterials/Base_Material_Functions/Fortnite_HideElementsWithVertexColor_2022}
  }
  MaterialExpressionMaterialFunctionCall_25 {
   class=MaterialExpressionMaterialFunctionCall
   inputs [Color {from=MaterialExpressionMaterialFunctionCall_13, output=Result}]
   properties {materialFunction=/Game/Packages/Fortress_SharedMaterials/Anime_Master_Material/MF_TODColorAdjustment}
  }
  MaterialExpressionMaterialFunctionCall_3 {
   class=MaterialExpressionMaterialFunctionCall
   inputs [In {from=MaterialExpressionNamedRerouteUsage_3}]
   properties {materialFunction=/Game/Packages/Fortress_SharedMaterials/Anime_Master_Material/MF_TODCustomColors}
  }
  MaterialExpressionMaterialFunctionCall_4 {
   class=MaterialExpressionMaterialFunctionCall
   inputs [
    Fade Offset {from=MaterialExpressionScalarParameter_39}
    Fade Distance {from=MaterialExpressionScalarParameter_38}
   ]
   properties {materialFunction=/Game/Global/Materials/Character_Parent_2023/Anime_Parent_2023/MaterialFunctions/MFC_DistanceFade_FOV_compensation}
  }
  MaterialExpressionMaterialFunctionCall_5 {
   class=MaterialExpressionMaterialFunctionCall
   inputs [
    Normal {from=MaterialExpressionTextureSampleParameter2D_4, output=RGB}
    Flatness {from=MaterialExpressionScalarParameter_0}
   ]
   properties {materialFunction=&FlattenNormal}
  }
  MaterialExpressionMin_0 {
   class=MaterialExpressionMin
   inputs [
    A {from=MaterialExpressionNamedRerouteUsage_6, output=""}
    B {from=MaterialExpressionReroute_0, output=""}
   ]
   properties {}
  }
  MaterialExpressionMultiply_0 {
   class=MaterialExpressionMultiply
   inputs [
    A {from=MaterialExpressionNormalize_2}
    B {from=MaterialExpressionScalarParameter_2}
   ]
  }
  MaterialExpressionMultiply_10 {
   class=MaterialExpressionMultiply
   inputs [
    A {from=MaterialExpressionVertexColor_0}
    B {from=""}
   ]
   properties {constB=2}
  }
  MaterialExpressionMultiply_2 {
   class=MaterialExpressionMultiply
   inputs [
    A {from=MaterialExpressionCeil_0}
    B {from=MaterialExpressionOneMinus_1}
   ]
  }
  MaterialExpressionMultiply_21 {
   class=MaterialExpressionMultiply
   inputs [
    A {from=MaterialExpressionStaticSwitchParameter_6}
    B {from=MaterialExpressionNamedRerouteUsage_1}
   ]
  }
  MaterialExpressionMultiply_22 {
   class=MaterialExpressionMultiply
   inputs [
    A {from=MaterialExpressionTextureSampleParameter2D_0, output=RGB}
    B {from=MaterialExpressionScalarParameter_34}
   ]
  }
  MaterialExpressionMultiply_3 {
   class=MaterialExpressionMultiply
   inputs [
    A {from=MaterialExpressionMultiply_4}
    B {from=MaterialExpressionOneMinus_0}
   ]
  }
  MaterialExpressionMultiply_4 {
   class=MaterialExpressionMultiply
   inputs [
    A {from=MaterialExpressionVertexColor_1, output=R}
    B {from=""}
   ]
   properties {constB=2}
  }
  MaterialExpressionMultiply_6 {
   class=MaterialExpressionMultiply
   inputs [
    A {from=MaterialExpressionNamedRerouteUsage_7}
    B {from=MaterialExpressionStaticSwitchParameter_12}
   ]
  }
  MaterialExpressionNamedRerouteDeclaration_0 {
   class=MaterialExpressionNamedRerouteDeclaration
   inputs [None {from=MaterialExpressionMaterialFunctionCall_4, output=Result}]
   properties {name=DistanceFading, nodeColor=<rgba 1 0.3047 0 1>}
  }
  MaterialExpressionNamedRerouteDeclaration_1 {
   class=MaterialExpressionNamedRerouteDeclaration
   inputs [None {from=MaterialExpressionPower_1}]
   properties {name=BakedNormalFresnel, nodeColor=<rgba 0 0.8144 1 1>}
  }
  MaterialExpressionNamedRerouteDeclaration_2 {
   class=MaterialExpressionNamedRerouteDeclaration
   inputs [None {from=MaterialExpressionStaticSwitchParameter_7}]
   properties {name=WPO, nodeColor=<rgba 1 0.3281 0 1>}
  }
  MaterialExpressionNamedRerouteDeclaration_3 {
   class=MaterialExpressionNamedRerouteDeclaration
   inputs [None {from=MaterialExpressionStaticSwitchParameter_3}]
   properties {name=OpacityMask, nodeColor=<rgba 0 0.8125 1 1>}
  }
  MaterialExpressionNamedRerouteDeclaration_4 {
   class=MaterialExpressionNamedRerouteDeclaration
   inputs [None {from=MaterialExpressionNormalize_0}]
   properties {name=BakedNormals, nodeColor=<rgba 0 0.2789 1 1>}
  }
  MaterialExpressionNamedRerouteDeclaration_5 {
   class=MaterialExpressionNamedRerouteDeclaration
   inputs [None {from=MaterialExpressionAdd_3}]
   properties {name=ThicknessMask, nodeColor=<rgba 1 0.1689 0 1>}
  }
  MaterialExpressionNamedRerouteDeclaration_6 {
   class=MaterialExpressionNamedRerouteDeclaration
   inputs [None {from=MaterialExpressionStaticSwitchParameter_5}]
   properties {name=Normal, nodeColor=<rgba 0.224 0.224 1 1>}
  }
  MaterialExpressionNamedRerouteDeclaration_7 {
   class=MaterialExpressionNamedRerouteDeclaration
   inputs [None {from=MaterialExpressionVectorParameter_4, output=RGB}]
   properties {name=ColorRaw, nodeColor=<rgba 0 0.5781 1 1>}
  }
  MaterialExpressionNamedRerouteDeclaration_8 {
   class=MaterialExpressionNamedRerouteDeclaration
   inputs [None {from=MaterialExpressionMaterialFunctionCall_25, output=Result}]
   properties {name=ColorFinal, nodeColor=<rgba 1 0.6797 0 1>}
  }
  MaterialExpressionNamedRerouteUsage_0 {class=MaterialExpressionNamedRerouteUsage}
  MaterialExpressionNamedRerouteUsage_1 {class=MaterialExpressionNamedRerouteUsage}
  MaterialExpressionNamedRerouteUsage_10 {class=MaterialExpressionNamedRerouteUsage}
  MaterialExpressionNamedRerouteUsage_11 {class=MaterialExpressionNamedRerouteUsage}
  MaterialExpressionNamedRerouteUsage_3 {class=MaterialExpressionNamedRerouteUsage}
  MaterialExpressionNamedRerouteUsage_4 {class=MaterialExpressionNamedRerouteUsage}
  MaterialExpressionNamedRerouteUsage_6 {class=MaterialExpressionNamedRerouteUsage}
  MaterialExpressionNamedRerouteUsage_7 {class=MaterialExpressionNamedRerouteUsage}
  MaterialExpressionNamedRerouteUsage_9 {class=MaterialExpressionNamedRerouteUsage}
  MaterialExpressionNormalize_0 {class=MaterialExpressionNormalize, inputs [VectorInput {from=MaterialExpressionSubtract_2, output=""}]}
  MaterialExpressionNormalize_1 {class=MaterialExpressionNormalize, inputs [VectorInput {from=MaterialExpressionMaterialFunctionCall_5, output=Result}]}
  MaterialExpressionNormalize_2 {class=MaterialExpressionNormalize, inputs [VectorInput {from=MaterialExpressionSubtract_0, output=""}]}
  MaterialExpressionObjectPositionWS_1 {class=MaterialExpressionObjectPositionWS, inputs [], properties {}}
  MaterialExpressionOneMinus_0 {class=MaterialExpressionOneMinus, inputs [None {from=MaterialExpressionVertexColor_1, output=A}]}
  MaterialExpressionOneMinus_1 {class=MaterialExpressionOneMinus, inputs [None {from=MaterialExpressionTwoSidedSign_0}]}
  MaterialExpressionOneMinus_2 {class=MaterialExpressionOneMinus, inputs [None {from=MaterialExpressionAbs_1}]}
  MaterialExpressionPinBase_0 {class=MaterialExpressionPinBase, inputs [], properties {}}
  MaterialExpressionPinBase_1 {
   class=MaterialExpressionPinBase
   inputs [
    M {from=MaterialExpressionConstant_1, output=""}
    Pre Skinned Local Position {from=MaterialExpressionShaderStageSwitch_0, output=""}
    Distance Blend {from=MaterialExpressionMaterialFunctionCall_14, output=Distance Blend}
    Fresnel {from=MaterialExpressionFresnel_0, output=""}
   ]
   properties {reroutePins [
     M {expression=&self:MaterialGraph_1.MaterialGraphNode_Composite_0.Construct Base Inputs.MaterialExpressionReroute_8}
     Pre Skinned Local Position {expression=&self:MaterialGraph_1.MaterialGraphNode_Composite_0.Construct Base Inputs.MaterialExpressionReroute_9}
     Distance Blend {expression=&self:MaterialGraph_1.MaterialGraphNode_Composite_0.Construct Base Inputs.MaterialExpressionReroute_10}
     Fresnel {expression=&self:MaterialGraph_1.MaterialGraphNode_Composite_0.Construct Base Inputs.MaterialExpressionReroute_11}
    ]}
  }
  MaterialExpressionPower_1 {
   class=MaterialExpressionPower
   inputs [
    Base {from=MaterialExpressionOneMinus_2}
    Exp {from=MaterialExpressionScalarParameter_1}
   ]
  }
  MaterialExpressionPreSkinnedPosition_0 {class=MaterialExpressionPreSkinnedPosition, inputs [], properties {}}
  MaterialExpressionReroute_0 {class=MaterialExpressionReroute, inputs [None {from=MaterialExpressionStaticSwitchParameter_3}]}
  MaterialExpressionReroute_1 {class=MaterialExpressionReroute, inputs [None {from=MaterialExpressionOneMinus_1}]}
  MaterialExpressionReroute_3 {class=MaterialExpressionReroute, inputs [None {from=MaterialExpressionMultiply_21}]}
  MaterialExpressionReroute_4 {class=MaterialExpressionReroute, inputs [None {from=MaterialExpressionScalarParameter_36}]}
  MaterialExpressionReroute_5 {class=MaterialExpressionReroute, inputs [None {from=MaterialExpressionReroute_6}]}
  MaterialExpressionReroute_6 {class=MaterialExpressionReroute, inputs [None {from=MaterialExpressionStaticSwitchParameter_1}]}
  MaterialExpressionSaturate_0 {class=MaterialExpressionSaturate, inputs [None {from=MaterialExpressionMultiply_3}]}
  MaterialExpressionScalarParameter_0 {
   class=MaterialExpressionScalarParameter
   properties {
    defaultValue=-0.1
    desc="values above 1.0 and below 0.0 can help with seperated geometry.\r\n-0.1 seems like a good wellyou to alight split geo."
    group=01_InvertedHull_Outlines
    parameterName=NormalMap_Flatness
    sliderMax=2
    sliderMin=-1
    sortPriority=5
   }
  }
  MaterialExpressionScalarParameter_1 {
   class=MaterialExpressionScalarParameter
   properties {defaultValue=5, group=01_InvertedHull_Outlines, parameterName=BakedNormalOpacity_Exponent, sliderMax=12}
  }
  MaterialExpressionScalarParameter_2 {
   class=MaterialExpressionScalarParameter
   properties {defaultValue=0.3, group=01_InvertedHull_Outlines, parameterName=CameraDirectionOffset, sliderMax=2, sortPriority=3}
  }
  MaterialExpressionScalarParameter_34 {
   class=MaterialExpressionScalarParameter
   properties {
    defaultValue=1
    desc=Multiplier for the Outline Color Texture
    group=01_InvertedHull_Outlines
    parameterName=Color_Intensity
    sliderMax=1
    sortPriority=3
   }
  }
  MaterialExpressionScalarParameter_36 {
   class=MaterialExpressionScalarParameter
   properties {defaultValue=0.1, desc="LineThickness (Near)", group=01_InvertedHull_Outlines, parameterName=LineThickness, sliderMax=0.5, sortPriority=3}
  }
  MaterialExpressionScalarParameter_38 {
   class=MaterialExpressionScalarParameter
   properties {defaultValue=1000, group=01_InvertedHull_Outlines, parameterName=Fade_Distance, sliderMax=1500, sliderMin=500, sortPriority=5}
  }
  MaterialExpressionScalarParameter_39 {
   class=MaterialExpressionScalarParameter
   properties {defaultValue=50, group=01_InvertedHull_Outlines, parameterName=Fade_Offset, sliderMax=100, sortPriority=5}
  }
  MaterialExpressionScalarParameter_40 {
   class=MaterialExpressionScalarParameter
   properties {defaultValue=0.6, group=01_InvertedHull_Outlines, parameterName=LineThickness_Far, sliderMax=2, sortPriority=3}
  }
  MaterialExpressionSetMaterialAttributes_0 {
   class=MaterialExpressionSetMaterialAttributes
   inputs [
    MaterialAttributes {from="", output=""}
    Emissive Color {from=MaterialExpressionNamedRerouteUsage_4, output=""}
    Opacity Mask {from=MaterialExpressionNamedRerouteUsage_9, output=""}
    World Position Offset {from=MaterialExpressionNamedRerouteUsage_10, output=""}
   ]
   properties {attributeSetTypes [#1, #2, #3]}
  }
  MaterialExpressionShaderStageSwitch_0 {
   class=MaterialExpressionShaderStageSwitch
   inputs [
    PixelShader {from=MaterialExpressionVertexInterpolator_0, output=PS}
    VertexShader {from=MaterialExpressionPreSkinnedPosition_0, output=""}
   ]
   properties {}
  }
  MaterialExpressionStaticSwitchParameter_0 {
   class=MaterialExpressionStaticSwitchParameter
   inputs [
    True {from=MaterialExpressionTransform_0, output=""}
    False {from="", output=""}
   ]
   properties {desc=use normal map to prevent segmented geometry, group=01_InvertedHull_Outlines, parameterName=UseNormalMap, sortPriority=0}
  }
  MaterialExpressionStaticSwitchParameter_1 {
   class=MaterialExpressionStaticSwitchParameter
   inputs [
    True {from=MaterialExpressionMaterialFunctionCall_2, output=Result}
    False {from=MaterialExpressionReroute_3, output=""}
   ]
   properties {group=03_Opacity Mask, parameterName=Use Vertex Colors for Mask}
  }
  MaterialExpressionStaticSwitchParameter_11 {
   class=MaterialExpressionStaticSwitchParameter
   inputs [
    True {from=MaterialExpressionMultiply_22, output=""}
    False {from=MaterialExpressionVectorParameter_4, output=RGB}
   ]
   properties {
    desc=Enable to use a texture to color the edge outlines of the mesh.
    group=01_InvertedHull_Outlines
    parameterName=UseColorMap
    sortPriority=0
   }
  }
  MaterialExpressionStaticSwitchParameter_12 {
   class=MaterialExpressionStaticSwitchParameter
   inputs [
    True {from=MaterialExpressionLinearInterpolate_8, output=""}
    False {from=MaterialExpressionScalarParameter_36, output=""}
   ]
   properties {defaultValue=true, group=01_InvertedHull_Outlines, parameterName=UseDistanceFading, sortPriority=0}
  }
  MaterialExpressionStaticSwitchParameter_3 {
   class=MaterialExpressionStaticSwitchParameter
   inputs [
    True {from=MaterialExpressionMultiply_2, output=""}
    False {from=MaterialExpressionReroute_1, output=""}
   ]
   properties {group=01_InvertedHull_Outlines, parameterName=UseThicknessMap, sortPriority=0}
  }
  MaterialExpressionStaticSwitchParameter_4 {
   class=MaterialExpressionStaticSwitchParameter
   inputs [
    True {from=MaterialExpressionMin_0, output=""}
    False {from=MaterialExpressionReroute_0, output=""}
   ]
   properties {group=01_InvertedHull_Outlines, parameterName=UseBakedNormalOpacityFresnel, sortPriority=0}
  }
  MaterialExpressionStaticSwitchParameter_5 {
   class=MaterialExpressionStaticSwitchParameter
   inputs [
    True {from=MaterialExpressionTransform_1, output=""}
    False {from=MaterialExpressionVertexNormalWS_0, output=""}
   ]
   properties {
    desc=use baked normals in the vertex colors instead of the surface normals.
    group=01_InvertedHull_Outlines
    parameterName=UseBakedNormals
    sortPriority=0
   }
  }
  MaterialExpressionStaticSwitchParameter_6 {
   class=MaterialExpressionStaticSwitchParameter
   inputs [
    True {from=MaterialExpressionMultiply_6, output=""}
    False {from=MaterialExpressionStaticSwitchParameter_12, output=""}
   ]
   properties {group=01_InvertedHull_Outlines, parameterName=UseThicknessMap, sortPriority=0}
  }
  MaterialExpressionStaticSwitchParameter_7 {
   class=MaterialExpressionStaticSwitchParameter
   inputs [
    True {from=MaterialExpressionIf_0, output=""}
    False {from=MaterialExpressionStaticSwitchParameter_8, output=""}
   ]
   properties {group=01_InvertedHull_Outlines, parameterName=HideOutlinesWithVertexColors, sortPriority=0}
  }
  MaterialExpressionStaticSwitchParameter_8 {
   class=MaterialExpressionStaticSwitchParameter
   inputs [
    True {from=MaterialExpressionAdd_0, output=""}
    False {from=MaterialExpressionReroute_5, output=""}
   ]
   properties {group=01_InvertedHull_Outlines, parameterName=OffsetOutlineInCameraDirection, sortPriority=90}
  }
  MaterialExpressionSubtract_0 {
   class=MaterialExpressionSubtract
   inputs [
    A {from=MaterialExpressionCameraPositionWS_0}
    B {from=MaterialExpressionWorldPosition_0, output=XYZ}
   ]
  }
  MaterialExpressionSubtract_1 {
   class=MaterialExpressionSubtract
   inputs [
    A {from=MaterialExpressionObjectPositionWS_1}
    B {from=MaterialExpressionAdd_4}
   ]
  }
  MaterialExpressionSubtract_2 {
   class=MaterialExpressionSubtract
   inputs [
    A {from=MaterialExpressionMultiply_10}
    B {from=""}
   ]
  }
  MaterialExpressionTextureSampleParameter2D_0 {
   class=MaterialExpressionTextureSampleParameter2D
   inputs [
    UVs {from="", output=""}
    Apply View MipBias {from="", output=""}
   ]
   properties {
    desc="Outline Color Texture, use can use the ShadedMap."
    group=01_InvertedHull_Outlines
    parameterName=ColorMap
    sortPriority=1
    texture=&T_BlacksRGB
   }
  }
  MaterialExpressionTextureSampleParameter2D_3 {
   class=MaterialExpressionTextureSampleParameter2D
   inputs [
    UVs {from="", output=""}
    Apply View MipBias {from="", output=""}
   ]
   properties {
    group=01_InvertedHull_Outlines
    parameterName=FX
    samplerType=SAMPLERTYPE_Masks
    sortPriority=1
    texture=/Game/Packages/Fortress_SharedMaterials/Anime_Master_Material/T_Outline_Mask
   }
  }
  MaterialExpressionTextureSampleParameter2D_4 {
   class=MaterialExpressionTextureSampleParameter2D
   inputs [
    UVs {from="", output=""}
    Apply View MipBias {from="", output=""}
   ]
   properties {
    group=01_InvertedHull_Outlines
    parameterName=NormalMap
    samplerType=SAMPLERTYPE_Normal
    sortPriority=1
    texture=&FlatNormalBlue
   }
  }
  MaterialExpressionTransform_0 {class=MaterialExpressionTransform, inputs [None {from=MaterialExpressionNormalize_1}]}
  MaterialExpressionTransform_1 {
   class=MaterialExpressionTransform
   inputs [None {from=MaterialExpressionNamedRerouteDeclaration_4}]
   properties {transformSourceType=TRANSFORMSOURCE_Local}
  }
  MaterialExpressionTransform_2 {
   class=MaterialExpressionTransform
   inputs [None {from=MaterialExpressionNamedRerouteUsage_11}]
   properties {transformSourceType=TRANSFORMSOURCE_Local}
  }
  MaterialExpressionTwoSidedSign_0 {class=MaterialExpressionTwoSidedSign, inputs [], properties {}}
  MaterialExpressionVectorParameter_4 {
   class=MaterialExpressionVectorParameter
   properties {group=01_InvertedHull_Outlines, parameterName=Color, sortPriority=2, defaultValue=<rgba 0.005 0.005 0.005 1>}
  }
  MaterialExpressionVertexColor_0 {class=MaterialExpressionVertexColor, inputs [], properties {}}
  MaterialExpressionVertexColor_1 {class=MaterialExpressionVertexColor, inputs [], properties {}}
  MaterialExpressionVertexInterpolator_0 {class=MaterialExpressionVertexInterpolator, inputs [VS {from=MaterialExpressionPreSkinnedPosition_0, output=""}], properties {}}
  MaterialExpressionVertexNormalWS_0 {class=MaterialExpressionVertexNormalWS, inputs [], properties {}}
  MaterialExpressionWorldPosition_0 {class=MaterialExpressionWorldPosition, inputs [], properties {}}
  MaterialExpressionWorldPosition_1 {class=MaterialExpressionWorldPosition, inputs [], properties {}}
 }
 properties {
  bAutomaticallySetUsageInEditor=false
  bEnableResponsiveAA=true
  bUseMaterialAttributes=true
  bUsedWithMeshDeformer=true
  bUsedWithMorphTargets=true
  bUsedWithSkeletalMesh=true
  blendMode=BLEND_Masked
  pixelDepthOffsetMode=PDOM_Legacy
  previewMesh=/BRCosmetics/Accessories/FORT_Backpacks/Backpack_M_MED_ShakeCrunch/Meshes/M_MED_ShakeCrunch_Pack
  shadingModel=MSM_Unlit
  twoSided=true
  usedShadingModels=Unlit
 }
}
