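# Masked unlit inverted-hull outline material for the ShakeCrunch medium male character, drawing a colored, irregular line.
# World position offset pushes vertices along normals, with optional thickness-map scaling and a vertex-color switch for hiding outlines.
# A view-angle gradient, distorted by a flowmap, cuts the line through the opacity mask.
# Virtual-plane coordinates, stepped time and distance-fade reroutes support the animated breakup, and results pass through the character parent PostFX function.
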Material /BRCosmetics/Characters/Player/Male/Medium/Bodies/M_MED_ShakeCrunch/Materials/M_ShakeCrunch_LoseOutline
graph {
 properties {
  bAutomaticallySetUsageInEditor=false
  bUseMaterialAttributes=true
  bUsedWithMeshDeformer=true
  bUsedWithMorphTargets=true
  bUsedWithSkeletalMesh=true
  bUsedWithStaticLighting=true
  blendMode=BLEND_Masked
  shadingModel=MSM_Unlit
  usedShadingModels=Unlit
 }
 nodes [
  Composite() {subgraphName=Construct Base Inputs}
  n1 = CameraVectorWS()
  n2 = Saturate(None=DotProduct(A=VertexNormalWS(), B=@n1))
  n3 = Multiply(A=ScalarParameter() {defaultValue=1.2, group=01_InvertedHull_Outlines, parameterName=LineMask_InnerIntensity, sliderMax=2}, B=@n2)
  n4 = Power(Base=@n2, Exp=ScalarParameter() {defaultValue=1, group=01_InvertedHull_Outlines, parameterName=LineMask_Power, sliderMax=4})
  n5 = Add(A=NamedRerouteUsage(), B=NamedRerouteUsage())
  n6 = Multiply(
   A=ComponentMask(None=TextureSample(UVs=@n5) {samplerType=SAMPLERTYPE_LinearColor, texture=/Game/Global/Textures/Default/Flowmaps/T_TorusFlowmap}.RGB) {g=true, r=true}
   B=ScalarParameter() {defaultValue=0.6, group=01_InvertedHull_Outlines, parameterName=Distortion_Amount, sliderMax=2}
  )
  n7 = LinearInterpolate(
   A=LinearInterpolate(
    A=ScalarParameter() {defaultValue=-6, group=01_InvertedHull_Outlines, parameterName=LineMask_Near, sliderMax=-10}
    B=ScalarParameter() {defaultValue=-1, group=01_InvertedHull_Outlines, parameterName=LineMask_Far, sliderMax=-10}
    Alpha=NamedRerouteUsage()
   )
   B=ScalarParameter() {defaultValue=1.5, group=01_InvertedHull_Outlines, parameterName=LineMask_B, sliderMax=3}
   Alpha=TextureSample(UVs=Add(A=@n4, B=@n6)) {texture=/Game/Global/Textures/Default/Masks_and_Gradients/gradient_stasis}.R
  ) {constA=-2, constB=2}
  n8 = Multiply(
   A=ScalarParameter() {defaultValue=1, group=01_InvertedHull_Outlines, parameterName=Outline_Opacity, sliderMax=2}
   B=Multiply(A=OneMinus(None=@n3), B=Saturate(None=@n7))
  )
  n9 = VertexColor()
  n10 = OneMinus(None=@n9.A)
  n11 = WorldPosition()
  n12 = Constant3Vector()
  n13 = TextureSampleParameter2D() {
   group=01_InvertedHull_Outlines
   parameterName=FX
   samplerType=SAMPLERTYPE_Masks
   sortPriority=1
   texture=/Game/Packages/Fortress_SharedMaterials/Anime_Master_Material/T_Outline_Mask
  }
  n14 = ChannelMaskParameter(None=@n13.RGB) {group=01_InvertedHull_Outlines, maskChannel=Blue, parameterName=OutlineThickness_Channel, sortPriority=3, defaultValue=<rgba 0 0 1 0>}
  n15 = Add(A=@n14.R, B=Constant() {desc=one bit used for opacity cutout, r=-0.009})
  n16 = Constant() {r=1}
  n17 = Multiply(
   A=VertexNormalWS()
   B=Multiply(
    A=StaticSwitchParameter(True=@n15, False=@n16) {group=01_InvertedHull_Outlines, parameterName=UseThicknessMap, sortPriority=0}
    B=ScalarParameter() {defaultValue=1, group=01_InvertedHull_Outlines, parameterName=LineThickness, sliderMax=2, sortPriority=3}
   )
  )
  n18 = If(A=Saturate(None=Multiply(A=@n9.R, B=@n10)), "A > B"=Subtract(A=ObjectPositionWS(), B=Add(A=@n11.XYZ, B=@n12)), "A < B"=@n17) {constB=1}
  n19 = Constant()
  MaterialExpressionReroute_4 = Reroute(None=@n19)
  n20 = PreSkinnedPosition()
  n21 = ShaderStageSwitch(PixelShader=VertexInterpolator(VS=@n20).PS, VertexShader=@n20)
  MaterialExpressionReroute_5 = Reroute(None=@n21)
  n22 = MaterialFunctionCall!(Material) {materialFunction=/Game/Packages/Fortress_SharedMaterials/Base_Material_Functions/CharacterBrightnessandScale_Balance}
  MaterialExpressionReroute_6 = Reroute(None=@n22.Distance Blend)
  n23 = Fresnel() {baseReflectFraction=0, exponent=1}
  MaterialExpressionReroute_7 = Reroute(None=@n23)
  MaterialFunctionCall(
   Material=SetMaterialAttributes!(
    MaterialAttributes
    Emissive Color=VectorParameter() {group=01_InvertedHull_Outlines, parameterName=Outline_Color, defaultValue=<rgba 0 0 0 1>}.RGB
    Opacity Mask=Saturate(None=@n8)
    World Position Offset=StaticSwitchParameter(True=@n18, False=@n17) {group=01_InvertedHull_Outlines, parameterName=HideOutlinesWithVertexColors, sortPriority=0}
   ) {attributeSetTypes [#1, #2, #3]}
   M=@MaterialExpressionReroute_4.M
   Pre Skinned Local Position=@MaterialExpressionReroute_5.Pre Skinned Local Position
   Distance Blend=@MaterialExpressionReroute_6.Distance Blend
   Fresnel=@MaterialExpressionReroute_7.Fresnel
  ) {materialFunction=&MF_CharacterParent_2023_PostFX}
  n24 = TwoSidedSign()
  Multiply(A=Saturate(None=OneMinus(None=@n24)))
  Multiply(A=@n1, B=Constant() {r=-1})
  NamedRerouteDeclaration(None=MaterialFunctionCall!(
    Fade Offset=ScalarParameter() {defaultValue=50, group=01_InvertedHull_Outlines, parameterName=Fade_Offset, sliderMax=100}
    Fade Distance=ScalarParameter() {defaultValue=1000, group=01_InvertedHull_Outlines, parameterName=Fade_Distance, sliderMax=1500, sliderMin=500}
   ) {materialFunction=/Game/Global/Materials/Character_Parent_2023/Anime_Parent_2023/MaterialFunctions/MFC_DistanceFade_FOV_compensation}.Result) {name=DistanceFading, nodeColor=<rgba 1 0.3047 0 1>}
  n25 = Constant3Vector() {constant=<rgba 1 1 0 0>}
  n26 = MaterialFunctionCall() {materialFunction=&CameraDirectionVector}
  n27 = Multiply(A=Constant() {desc=Needs at least a small amount of this or it will break, r=0.25}, B=Normalize(VectorInput=Multiply(A=@n25, B=@n26.Result)))
  n28 = Multiply(A=@n26.Result, B=ScalarParameter() {defaultValue=70, group=01_InvertedHull_Outlines, parameterName=VP_CameraVectorMin, sliderMax=100})
  n29 = ObjectPositionWS()
  NamedRerouteDeclaration(None=MaterialFunctionCall(
    Plane Normal Axis=Add(A=@n27, B=@n26.Result)
    Plane Center=Add(A=@n28, B=@n29)
    UV Scale=ScalarParameter() {defaultValue=50, group=01_InvertedHull_Outlines, parameterName=VP_Scale}
   ) {materialFunction=&VirtualPlaneCoordinates}.UVs) {name=VirtualPlane, nodeColor=<rgba 1 0.9141 0 1>}
  NamedRerouteDeclaration(None=MaterialFunctionCall() {materialFunction=&/MFC_ShakeCrunch_SteppedTime}.TimeStep) {name=TimeStep, nodeColor=<rgba 0.052 1 0.9901 1>}
  PinBase!(M=@n19, Pre Skinned Local Position=@n21, Distance Blend=@n22.Distance Blend, Fresnel=@n23) {reroutePins [
    M {expression=&self:MaterialExpressionReroute_4}
    Pre Skinned Local Position {expression=&self:MaterialExpressionReroute_5}
    Distance Blend {expression=&self:MaterialExpressionReroute_6}
    Fresnel {expression=&self:MaterialExpressionReroute_7}
   ]}
  PinBase!()
 ]
}
